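# Master material for the Nico1Radiant vehicle cosmetic wheel, adding animated emissive glow to rim and tire on the simple wheel parent function.
# Noise-distorted plasma, scrolling pipe pulses and rotating radial-gradient tire patterns are built as named reroutes.
# Rim and tire emissives combine masks, intensity controls and normal-based Fresnel, with a Painted switch selecting colors, then blend together.
# The result passes through vehicle damage and de-rez transition functions.
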Material /VehicleCosmetics/Wheels/Nico1Radiant/Materials/M_Wheel_Nico1Radiant
graph {
 properties {
  bAutomaticallySetUsageInEditor=false
  bUseMaterialAttributes=true
  bUsedWithMeshDeformer=true
  bUsedWithSkeletalMesh=true
  pixelDepthOffsetMode=PDOM_Legacy
  previewMesh=/VehicleCosmetics/Wheels/Nico1Radiant/Meshes/SK_Wheel_Nico1Radiant
  usedShadingModels=Default Lit
 }
 nodes [
  Constant()
  Constant() {r=1}
  n1 = GetMaterialAttributes(None=MaterialFunctionCall() {materialFunction=&MF_WheelParent_Simple}.MaterialAttributes) {attributeGetTypes [#1]}
  n2 = NamedRerouteUsage()
  LinearInterpolate(A=@n1.Emissive Color, B=@n2)
  n3 = Add(A=@n1.Emissive Color, B=@n2)
  MaterialFunctionCall(In=MaterialFunctionCall(Material=SetMaterialAttributes(MaterialAttributes=@n1.MaterialAttributes, Emissive Color=@n3) {attributeSetTypes [#1]}) {materialFunction=&MF_BaseVehicle_Damage}.Result) {materialFunction=&MF_VehicleDeRez_Transition}
  n4 = VectorParameter() {group=5_Special Effects, parameterName=EmissiveColor_Tire, defaultValue=<rgba 0 0.3613 1 1>}
  NamedRerouteDeclaration(None=@n4.RGB) {name=EmissiveColorTire, nodeColor=<rgba 0.2187 1 0 1>}
  n5 = TextureSampleParameter2D() {
   group=5_Special Effects
   parameterName=Tire_RGB_M
   samplerSource=SSM_Wrap_WorldGroupSettings
   samplerType=SAMPLERTYPE_LinearColor
   texture=/VehicleCosmetics/Wheels/Nico1Radiant/Textures/T_Nico1Radiant_Tire_RGB
  }
  NamedRerouteDeclaration(None=@n5.B) {name=Tire_M_Blue_CH, nodeColor=<rgba 0.0078 0 1 1>}
  n6 = VectorParameter() {group=5_Special Effects, parameterName=EmissivePaintColor2, defaultValue=<rgba 0.491 0.1248 0.1195 1>}
  n7 = StaticSwitchParameter(True=@n6.RGB, False=VectorParameter() {group=5_Special Effects, parameterName=EmissiveColor2, defaultValue=<rgba 0.491 0.491 0.491 1>}.RGB) {group=5_Special Effects, parameterName=Painted}
  n8 = NamedRerouteUsage()
  n9 = NamedRerouteUsage()
  n10 = Add(
   A=Multiply(A=@n7, B=@n8)
   B=Multiply(A=ScalarParameter() {defaultValue=1, group=5_Special Effects, parameterName=EmissiveRim_AddHighlights}, B=@n9)
  )
  n11 = NamedRerouteUsage()
  n12 = Multiply(
   A=Add(A=Multiply(A=@n10, B=@n11), B=NamedRerouteUsage())
   B=ScalarParameter() {defaultValue=1, group=5_Special Effects, parameterName=EmissiveRimIntensity}
  )
  n13 = TextureSampleParameter2D() {
   group=5_Special Effects
   parameterName=Rim_Normal
   samplerSource=SSM_Wrap_WorldGroupSettings
   samplerType=SAMPLERTYPE_Normal
   texture=/VehicleCosmetics/Wheels/Nico1Radiant/Textures/T_Nico1Radiant_Rim_N
  }
  n14 = Constant() {r=2}
  n15 = Multiply(
   A=Multiply(
    A=MaterialFunctionCall!(Input Normal=@n13.RGB, Input Sharpness=@n14, Input Two Sided) {materialFunction=/VehicleCosmetics/Wheels/CarCar_Holographic/Materials/MF_Fresnel_Advanced}.Edge
    B=ScalarParameter() {defaultValue=0.5, group=5_Special Effects, parameterName=Fresnel_RimIntensity}
   )
   B=@n7
  )
  n16 = StaticSwitchParameter(True=@n6.RGB, False=@n4.RGB) {group=5_Special Effects, parameterName=Painted}
  n17 = Add(
   A=Multiply(A=@n16, B=@n8)
   B=Multiply(A=ScalarParameter() {defaultValue=1, group=5_Special Effects, parameterName=EmissiveTread_AddHighlights}, B=@n9)
  )
  n18 = NamedRerouteUsage()
  n19 = Multiply(
   A=Add(A=Multiply(A=@n17, B=@n18), B=NamedRerouteUsage())
   B=ScalarParameter() {defaultValue=1, group=5_Special Effects, parameterName=EmissiveTireIntensity}
  )
  n20 = TextureSampleParameter2D() {
   group=5_Special Effects
   parameterName=Tire_Normal
   samplerSource=SSM_Wrap_WorldGroupSettings
   samplerType=SAMPLERTYPE_Normal
   texture=/VehicleCosmetics/Wheels/Nico1Radiant/Textures/T_Nico1Radiant_Tire_N
  }
  n21 = Multiply(
   A=Multiply(
    A=MaterialFunctionCall(Input Normal=@n20.RGB, Input Sharpness=@n14) {materialFunction=&MF_Fresnel_Advanced}.Edge
    B=ScalarParameter() {defaultValue=0.5, group=5_Special Effects, parameterName=Fresnel_TireIntensity}
   )
   B=@n16
  )
  NamedRerouteDeclaration(None=LinearInterpolate(A=Add(A=@n12, B=@n15), B=Add(A=@n19, B=@n21), Alpha=NamedRerouteUsage())) {name=Emissive, nodeColor=<rgba 0.0547 0 1 1>}
  n22 = TextureSampleParameter2D() {
   group=5_Special Effects
   parameterName=Rim_RGB_M
   samplerSource=SSM_Wrap_WorldGroupSettings
   samplerType=SAMPLERTYPE_LinearColor
   texture=/VehicleCosmetics/Wheels/Nico1Radiant/Textures/T_Nico1Radiant_Rim_RGB
  }
  NamedRerouteDeclaration(None=@n22.B) {name=Rim_M_Blue_CH_1, nodeColor=<rgba 0.0078 0 1 1>}
  n23 = TextureCoordinate() {coordinateIndex=1}
  n24 = TextureCoordinate() {coordinateIndex=1}
  n25 = Panner(Coordinate=@n24, Time=Multiply(A=Time(), B=Constant() {r=0.5})) {speedX=0.2, speedY=0.1}
  n26 = Rotator(Coordinate=@n24)
  n27 = Multiply(
   A=TextureSample(UVs=@n25) {samplerSource=SSM_Wrap_WorldGroupSettings, texture=/VehicleCosmetics/Wheels/Nico1Radiant/Textures/T_Nico1Radiant_Rim_Noise_RGB}.G
   B=TextureSample(UVs=Panner(Coordinate=@n26) {speedX=-0.09, speedY=0.09}) {samplerSource=SSM_Wrap_WorldGroupSettings, texture=&T_Nico1Radiant_Rim_Noise_RGB}.G
  )
  n28 = Add(
   A=TextureSample(UVs=@n25) {samplerSource=SSM_Wrap_WorldGroupSettings, texture=&T_Nico1Radiant_Rim_Noise_RGB}.R
   B=Add(A=@n27, B=@n27)
  )
  n29 = Constant() {r=1}
  n30 = TextureSample(UVs=Add(A=Panner(Coordinate=@n23) {speedX=0.2, speedY=0.1}, B=Multiply(A=@n28, B=@n29))) {samplerSource=SSM_Wrap_WorldGroupSettings, texture=&T_Nico1Radiant_Rim_Noise_RGB}
  NamedRerouteDeclaration(None=Clamp(None=ConstantBiasScale(None=@n30.B) {bias=0.5, scale=1})) {name=Plasma1, nodeColor=<rgba 0 1 0.2969 1>}
  NamedRerouteDeclaration(None=Clamp(None=ConstantBiasScale(None=@n30.B) {bias=-0.1, scale=1})) {name=Plasma2, nodeColor=<rgba 0 1 0.2969 1>}
  n31 = TextureSample(UVs=Panner(Coordinate=TextureCoordinate() {uTiling=2}) {speedX=-0.5}) {texture=&T_Nico1Radiant_Rim_Noise_RGB}
  n32 = TextureSampleParameter2D(UVs=TextureCoordinate() {coordinateIndex=1}) {
   group=5_Special Effects
   parameterName=Rim_RGB_M
   samplerSource=SSM_Wrap_WorldGroupSettings
   samplerType=SAMPLERTYPE_LinearColor
   texture=&T_Nico1Radiant_Rim_RGB
  }
  NamedRerouteDeclaration(None=Multiply(A=Multiply(A=@n31.B, B=@n32.B), B=@n7)) {name=PipePulse, nodeColor=<rgba 0 0.3438 1 1>}
  NamedRerouteDeclaration(None=@n5.G) {name=Tire_M_Green_CH, nodeColor=<rgba 0 1 0.1555 1>}
  NamedRerouteDeclaration(None=@n5.R) {name=Tire_M_Red_CH, nodeColor=<rgba 1 0 0.0137 1>}
  n33 = TextureCoordinate()
  n34 = Constant() {r=0.4}
  n35 = TextureSample(UVs=Rotator(Coordinate=Panner(Coordinate=@n33, Time=@n34) {speedX=1})) {
   samplerSource=SSM_Wrap_WorldGroupSettings
   samplerType=SAMPLERTYPE_LinearColor
   texture=/VehicleCosmetics/Wheels/Nico1Radiant/Textures/T_Nico1Radiant_Rim_RadialGradient
  }
  n36 = TextureCoordinate()
  n37 = Constant() {r=-0.6}
  n38 = TextureSample(UVs=Rotator(Coordinate=Panner(Coordinate=@n36, Time=@n37))) {
   samplerSource=SSM_Wrap_WorldGroupSettings
   samplerType=SAMPLERTYPE_LinearColor
   texture=&T_Nico1Radiant_Rim_RadialGradient
  }
  n39 = TextureSample(UVs=@n36) {
   samplerSource=SSM_Wrap_WorldGroupSettings
   samplerType=SAMPLERTYPE_LinearColor
   texture=/VehicleCosmetics/Wheels/Nico1Radiant/Textures/T_Nico1Radiant_Tire_RGB2
  }
  n40 = Add(A=Multiply(A=NamedRerouteUsage(), B=@n35.R), B=Multiply(A=@n39.R, B=@n38.R))
  n41 = TextureCoordinate()
  n42 = Constant() {r=-0.05}
  n43 = TextureSample(UVs=Rotator(Coordinate=Panner(Coordinate=@n41, Time=@n42))) {
   samplerSource=SSM_Wrap_WorldGroupSettings
   samplerType=SAMPLERTYPE_LinearColor
   texture=&T_Nico1Radiant_Rim_RadialGradient
  }
  n44 = Multiply(A=NamedRerouteUsage(), B=@n43.B)
  NamedRerouteDeclaration(None=Multiply(A=Add(A=@n40, B=@n44), B=NamedRerouteUsage())) {name=TirePatternEmissive, nodeColor=<rgba 0 1 0.6953 1>}
  NamedRerouteDeclaration(None=VertexColor().R) {name=Red_VC, nodeColor=<rgba 1 0 0 1>}
 ]
}
